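# Material function that supplies shared frontend UI backing colors for normal, hover and selected states, with optional special event recoloring.
# Each output reads default and special colors from the MPC_FrontendUI parameter collection and lerps between them by the collection's UseSpecialEventColors value.
# A DisableSpecialEventUIColors static switch lets a material bypass the blend and use only the default colors.
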MaterialFunction /Game/Athena/UI/Frontend/Art/MF_UIColorCollection_Frontend
graph {nodes [
  n1 = CollectionParameter() {collection=&/MPC_FrontendUI, parameterName=Default_RGBA_BackingColor}
  n2 = CollectionParameter() {collection=&MPC_FrontendUI, parameterName=Special_RGBA_BackingColor}
  n3 = CollectionParameter() {collection=&MPC_FrontendUI, parameterName=UseSpecialEventColors}
  FunctionOutput(None=StaticSwitchParameter(True=@n1, False=LinearInterpolate(A=@n1, B=@n2, Alpha=@n3)) {parameterName=DisableSpecialEventUIColors}) {outputName=RGBA_BackingColor}
  n4 = CollectionParameter() {collection=&MPC_FrontendUI, parameterName=Default_RGBA_BackingColor_Hover}
  n5 = CollectionParameter() {collection=&MPC_FrontendUI, parameterName=Special_RGBA_BackingColor_Hover}
  FunctionOutput(None=StaticSwitchParameter(True=@n4, False=LinearInterpolate(A=@n4, B=@n5, Alpha=@n3)) {parameterName=DisableSpecialEventUIColors}) {outputName=RGBA_BackingColor_Hover, sortPriority=1}
  n6 = CollectionParameter() {collection=&MPC_FrontendUI, parameterName=Default_RGBA_BackingColor_Selected}
  n7 = CollectionParameter() {collection=&MPC_FrontendUI, parameterName=Special_RGBA_BackingColor_Selected}
  FunctionOutput(None=StaticSwitchParameter(True=@n6, False=LinearInterpolate(A=@n6, B=@n7, Alpha=@n3)) {parameterName=DisableSpecialEventUIColors}) {outputName=RGBA_BackingColor_Selected, sortPriority=2}
 ]}
