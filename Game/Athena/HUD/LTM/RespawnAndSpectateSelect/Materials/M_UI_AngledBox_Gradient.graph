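# Translucent UI material for the Athena LTM respawn and spectate select HUD, drawing a configurable angled box.
# Rotated-UV half-plane masks, sharpened by resolution correction, cut the outline, inner border and optional bolt-shaped top edge.
# These masks composite backing, border and optional texture, with radial gradients and parameter-driven hover and selected colors.
# Static switches add a timed shine swipe, screen-space alpha falloff, or a simple fill or linear gradient mode.
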Material /Game/Athena/HUD/LTM/RespawnAndSpectateSelect/Materials/M_UI_AngledBox_Gradient
graph {
 properties {blendMode=BLEND_Translucent, materialDomain=MD_UI, numCustomizedUVs=5, pixelDepthOffsetMode=PDOM_Legacy, usedShadingModels=Default Lit}
 nodes [
  n1 = VectorParameter() {group=Pulse, parameterName=PulseValues, defaultValue=<rgba 1 1 1 1>}
  n2 = Time()
  n3 = ScalarParameter() {defaultValue=1, group=Pulse, parameterName=PulseSpeed}
  n4 = Power(Base=Fmod(A=Multiply(A=@n2, B=@n3), B=Constant() {r=1}), Exp=ScalarParameter() {defaultValue=1, group=Pulse, parameterName=PulseFalloff})
  MaterialFunctionCall(UVs=TextureCoordinate(), Texture Scale=LinearInterpolate(A=@n1.R, B=@n1.R, Alpha=@n4)) {materialFunction=&ScaleUVsByCenter}
  n5 = ScalarParameter() {defaultValue=1, group=IsJustFill, parameterName=SimpleFillAlpha}
  n6 = StaticComponentMaskParameter(None=TextureCoordinate()) {defaultG=true, group=IsJustFill, parameterName=GradientOrientation}
  n7 = OneMinus(None=@n6)
  n8 = ScalarParameter() {group=IsJustFill, parameterName=FlipGradientDirection}
  n9 = Saturate(None=Power(
    Base=LinearInterpolate(A=@n6, B=@n7, Alpha=@n8)
    Exp=ScalarParameter() {defaultValue=1, group=IsJustFill, parameterName=LinearGradientFalloffRate}
   ))
  n10 = Saturate(None=LinearInterpolate(
    A=ScalarParameter() {group=IsJustFill, parameterName=SImpleGradient_MinAlpha}
    B=ScalarParameter() {defaultValue=1, group=IsJustFill, parameterName=SimpleGradient_MaxAlpha}
    Alpha=@n9
   ))
  n11 = ScalarParameter() {group=IsJustFill, parameterName=UseGradientAlpha}
  n12 = ScalarParameter() {parameterName=UseTexture2}
  n13 = MaterialFunctionCall() {materialFunction=&GetUserInterfaceUV}
  n14 = MaterialFunctionCall(Float2=@n13.Pixel Size) {materialFunction=&BreakOutFloat2Components}
  n15 = Divide(A=@n14.R, B=@n14.R)
  n16 = AppendVector(A=@n15, B=Constant() {r=1})
  n17 = Multiply(A=OneMinus(None=@n15)) {constB=0.5}
  n18 = Constant()
  n19 = StaticSwitchParameter(True=Add(A=Multiply(A=@n13.Normalized UV, B=@n16), B=AppendVector(A=@n17, B=@n18)), False=@n13.Normalized UV) {defaultValue=true, group=Texture, parameterName=UseSmartUVs}
  n20 = ScalarParameter() {group=Texture, parameterName=RotateTexture}
  n21 = ScalarParameter() {group=Texture, parameterName=Offset_Texture_U}
  n22 = ScalarParameter() {group=Texture, parameterName=Offset_Texture_V}
  n23 = MaterialFunctionCall(
   UVs=Add(A=Rotator(Coordinate=@n19, Time=@n20) {speed=6.2832}, B=AppendVector(A=@n21, B=@n22))
   Texture Scale=AppendVector(
    A=ScalarParameter() {defaultValue=1, group=Texture, parameterName=Scale_Texture_U}
    B=ScalarParameter() {defaultValue=1, group=Texture, parameterName=Scale_Texture_V}
   )
  ) {materialFunction=&ScaleUVsByCenter}
  n24 = ScalarParameter() {group=Texture2, parameterName=RotateTexture2}
  n25 = ScalarParameter() {group=Texture, parameterName=Offset_Texture2_U}
  n26 = ScalarParameter() {group=Texture, parameterName=Offset_Texture2_V}
  n27 = MaterialFunctionCall(
   UVs=Add(A=Rotator(Coordinate=@n19, Time=@n24) {speed=6.2832}, B=AppendVector(A=@n25, B=@n26))
   Texture Scale=AppendVector(
    A=ScalarParameter() {defaultValue=1, group=Texture, parameterName=Scale_Texture2_U}
    B=ScalarParameter() {defaultValue=1, group=Texture, parameterName=Scale_Texture2_V}
   )
  ) {materialFunction=&ScaleUVsByCenter}
  n28 = TextureSampleParameter2D(UVs=@n23.UVs) {
   group=Texture
   parameterName=Texture
   samplerSource=SSM_Clamp_WorldGroupSettings
   texture=/Game/UI/Foundation/Textures/Icons/Heroes/Athena/Soldier/T-Soldier-HID-005-Athena-Commando-M
  }
  n29 = ScalarParameter() {parameterName=UseWhiteAsAlpha}
  n30 = TextureSampleParameter2D(UVs=@n27.UVs) {
   group=Texture
   parameterName=Texture_2
   samplerSource=SSM_Clamp_WorldGroupSettings
   texture=/Game/UI/Frontend/Campaign/Textures/T-Reward-Banner
  }
  n31 = MaterialFunctionCall!(
   UVs=LinearInterpolate(A=@n23.UVs, B=@n27.UVs, Alpha=@n12)
   Alpha=LinearInterpolate(A=LinearInterpolate(A=@n28.A, B=@n28.A, Alpha=@n29), B=@n30.A, Alpha=@n12)
  ) {materialFunction=/Game/UI/Foundation/Items/ClampAlphaToTransparentBorder}
  n32 = TextureCoordinate()
  n33 = ScalarParameter() {defaultValue=-0.03, group=Border, parameterName=Border_Rotation_Left}
  n34 = ComponentMask(None=Add(
    A=Rotator(Coordinate=@n32, Time=@n33) {speed=6.2832}
    B=ScalarParameter() {defaultValue=-0.135, group=Border, parameterName=Border_Offset_Left}
   )) {r=true}
  n35 = MaterialFunctionCall!(Native_U_PixelSize=ScalarParameter() {defaultValue=128, group=ResolutionSharpness, parameterName=NativePixelSize}) {materialFunction=/FortUILibrary/Foundation/Art_ToolKit/Tools/M_UI_Resolution_Sharpness_Correction}
  n36 = Multiply(A=@n35.Result, B=ScalarParameter() {defaultValue=64, group=Border, parameterName=Border_Sharpness_U})
  n37 = Rotator(Coordinate=TextureCoordinate(), Time=ScalarParameter() {defaultValue=-0.03, group=Border, parameterName=Border_Rotation_Right}) {speed=6.2832}
  n38 = ComponentMask(None=Add(A=OneMinus(None=@n37), B=ScalarParameter() {defaultValue=-0.135, group=Border, parameterName=Border_Offset_Right})) {r=true}
  n39 = DotProduct(A=Saturate(None=Multiply(A=@n34, B=@n36) {constB=32}), B=Saturate(None=Multiply(A=@n38, B=@n36) {constB=32}))
  n40 = TextureCoordinate() {coordinateIndex=4}
  n41 = ScalarParameter() {defaultValue=-0.245, group=Border, parameterName=Border_Rotation_Top}
  n42 = ComponentMask(None=Add(
    A=Rotator(Coordinate=@n40, Time=@n41) {speed=6.2832}
    B=ScalarParameter() {defaultValue=-0.125, group=Border, parameterName=Border_Offset_Top}
   )) {r=true}
  n43 = Multiply(A=@n35.Result, B=ScalarParameter() {defaultValue=64, group=Border, parameterName=Border_Sharpness_V})
  n44 = Rotator(
   Coordinate=TextureCoordinate() {coordinateIndex=4}
   Time=ScalarParameter() {defaultValue=-0.245, group=Border, parameterName=Border_Rotation_Bottom}
  ) {speed=6.2832}
  n45 = ComponentMask(None=Add(A=OneMinus(None=@n44), B=ScalarParameter() {defaultValue=-0.125, group=Border, parameterName=Border_Offset_Bottom})) {r=true}
  n46 = DotProduct(A=Saturate(None=Multiply(A=@n42, B=@n43) {constB=32}), B=Saturate(None=Multiply(A=@n45, B=@n43) {constB=32}))
  n47 = Saturate(None=DotProduct(A=Saturate(None=@n39), B=Saturate(None=@n46)))
  n48 = DotProduct(A=@n31.Result, B=@n47)
  n49 = Saturate(None=@n48)
  n50 = Multiply(A=@n49, B=ScalarParameter() {defaultValue=1, group=Alpha, parameterName=TextureAlpha})
  n51 = Saturate(None=OneMinus(None=@n47))
  n52 = TextureCoordinate() {coordinateIndex=4}
  n53 = ScalarParameter() {defaultValue=-0.245, group=Bolt, parameterName=Shape_Rotation_RightBolt}
  n54 = ComponentMask(None=Add(
    A=Rotator(Coordinate=@n52, Time=@n53) {speed=6.2832}
    B=ScalarParameter() {defaultValue=-0.075, group=Bolt, parameterName=Shape_Offset_RightBolt}
   )) {r=true}
  n55 = Multiply(A=@n35.Result, B=ScalarParameter() {defaultValue=64, group=Bolt, parameterName=Shape_Sharpness_V})
  n56 = Add(
   A=Rotator(
    Coordinate=TextureCoordinate() {coordinateIndex=4}
    Time=ScalarParameter() {defaultValue=-0.2375, group=Bolt, parameterName=Shape_Rotation_LeftBolt}
   ) {speed=6.2832}
   B=ScalarParameter() {defaultValue=-0.1, group=Bolt, parameterName=Shape_Offset_LeftBolt}
  )
  n57 = ScalarParameter() {defaultValue=64, group=Bolt, parameterName=Shape_Sharpness_V}
  n58 = Saturate(None=Multiply(A=ComponentMask(None=@n56) {r=true}, B=Multiply(A=@n35.Result, B=@n57)) {constB=32})
  n59 = TextureCoordinate() {coordinateIndex=4}
  n60 = ScalarParameter() {defaultValue=-0.025, group=Bolt, parameterName=Shape_Rotation_CenterBolt}
  n61 = ComponentMask(None=Add(
    A=Rotator(Coordinate=@n59, Time=@n60) {speed=6.2832}
    B=ScalarParameter() {defaultValue=-0.5, group=Bolt, parameterName=Shape_Offset_CenterBolt}
   )) {r=true}
  n62 = Multiply(A=@n35.Result, B=ScalarParameter() {defaultValue=64, group=Bolt, parameterName=Shape_Sharpness_U})
  n63 = Saturate(None=OneMinus(None=Multiply(A=@n61, B=@n62) {constB=32}))
  n64 = Add(A=Saturate(None=Multiply(A=@n54, B=@n55) {constB=32}), B=Saturate(None=DotProduct(A=@n58, B=@n63)))
  n65 = Rotator(
   Coordinate=TextureCoordinate() {coordinateIndex=4}
   Time=ScalarParameter() {defaultValue=-0.245, group=OverallShape, parameterName=Shape_Rotation_Top}
  ) {speed=6.2832}
  n66 = ScalarParameter() {defaultValue=-0.075, group=OverallShape, parameterName=Shape_Offset_Top}
  n67 = Multiply(A=@n35.Result, B=ScalarParameter() {defaultValue=64, group=OverallShape, parameterName=Shape_Sharpness_V})
  n68 = Multiply(A=ComponentMask(None=Add(A=@n65, B=@n66)) {r=true}, B=@n67) {constB=32}
  n69 = Rotator(
   Coordinate=TextureCoordinate() {coordinateIndex=4}
   Time=ScalarParameter() {defaultValue=-0.245, group=OverallShape, parameterName=Shape_Rotation_Bottom}
  ) {speed=6.2832}
  n70 = ComponentMask(None=Add(A=OneMinus(None=@n69), B=ScalarParameter() {defaultValue=-0.075, group=OverallShape, parameterName=Shape_Offset_Bottom})) {r=true}
  n71 = DotProduct(
   A=StaticSwitchParameter(True=Saturate(None=@n64), False=Saturate(None=@n68)) {group=Bolt, parameterName=HasTopBolt}
   B=Saturate(None=Multiply(A=@n70, B=@n67) {constB=32})
  )
  n72 = TextureCoordinate()
  n73 = ScalarParameter() {defaultValue=-0.025, group=OverallShape, parameterName=Shape_Rotation_Left}
  n74 = ComponentMask(None=Add(
    A=Rotator(Coordinate=@n72, Time=@n73) {speed=6.2832}
    B=ScalarParameter() {defaultValue=-0.0875, group=OverallShape, parameterName=Shape_Offset_Left}
   )) {r=true}
  n75 = Multiply(A=@n35.Result, B=ScalarParameter() {defaultValue=64, group=OverallShape, parameterName=Shape_Sharpness_U})
  n76 = Rotator(Coordinate=TextureCoordinate(), Time=ScalarParameter() {defaultValue=-0.025, group=OverallShape, parameterName=Shape_Rotation_Right}) {speed=6.2832}
  n77 = ComponentMask(None=Add(A=OneMinus(None=@n76), B=ScalarParameter() {defaultValue=-0.0875, group=OverallShape, parameterName=Shape_Offset_Right})) {r=true}
  n78 = DotProduct(A=Saturate(None=Multiply(A=@n74, B=@n75) {constB=32}), B=Saturate(None=Multiply(A=@n77, B=@n75) {constB=32}))
  n79 = Saturate(None=DotProduct(A=Saturate(None=@n71), B=Saturate(None=@n78)))
  n80 = Multiply(A=DotProduct(A=@n79, B=@n51), B=ScalarParameter() {defaultValue=1, group=Alpha, parameterName=BorderAlpha})
  n81 = Add(A=@n80, B=Multiply(A=@n47, B=ScalarParameter() {defaultValue=1, group=Alpha, parameterName=BackingAlpha}))
  n82 = Multiply(A=@n79, B=ScalarParameter() {defaultValue=1, group=Alpha, parameterName=BackingAlpha})
  n83 = StaticSwitchParameter(
   True=StaticSwitchParameter(True=Saturate(None=@n81), False=Saturate(None=@n82)) {defaultValue=true, group=Border, parameterName=HasBorder}
   False=@n80
  ) {defaultValue=true, group=Alpha, parameterName=HasBacking}
  n84 = Saturate(None=Add(A=@n83, B=Saturate(None=@n50)))
  n85 = Saturate(None=StaticSwitchParameter(
    True=LinearInterpolate(A=@n5, B=@n10, Alpha=@n11)
    False=StaticSwitchParameter(True=@n84, False=@n83) {group=Texture, parameterName=HasTexture}
   ) {group=IsJustFill, parameterName=IsSimpleFill})
  n86 = ComponentMask(None=MaterialFunctionCall(Texture Dimensions=TextureCoordinate()) {materialFunction=&ScreenAlignedUVs}."X 100%, Y 100%") {r=true}
  n87 = OneMinus(None=@n86)
  n88 = Saturate(None=Multiply(A=DotProduct(A=@n87, B=@n86), B=ScalarParameter() {defaultValue=16, group=ScreenSpace, parameterName=FallOffPoint}))
  n89 = ScalarParameter() {defaultValue=1, group=ScreenSpace, parameterName=FallOffRate}
  n90 = DotProduct(A=@n85, B=Saturate(None=Power(Base=@n88, Exp=@n89)))
  Multiply(
   A=Saturate(None=StaticSwitchParameter(True=@n90, False=@n85) {group=ScreenSpace, parameterName=HasScreenSpaceTreatment})
   B=LinearInterpolate(A=@n1.B, B=@n1.B, Alpha=@n4)
  )
  n91 = StaticSwitchParameter(
   True=LinearInterpolate(
    A=VectorParameter() {group=IsJustFill, parameterName=SimpleFillColor, defaultValue=<rgba 0 1 0 1>}.RGB
    B=VectorParameter() {group=IsJustFill, parameterName=SimpleFillColor_2, defaultValue=<rgba 1 0 0 1>}.RGB
    Alpha=@n9
   )
   False=VectorParameter() {group=IsJustFill, parameterName=SimpleFillColor, defaultValue=<rgba 0 1 0 1>}.RGB
  ) {parameterName=IsSimpleGradient}
  n92 = Constant()
  n93 = VectorParameter() {group=Shine, parameterName=ShineRGBA, defaultValue=<rgba 1 1 1 0.5>}
  n94 = ScalarParameter() {defaultValue=4, group=Shine, parameterName=Shine Speed}
  n95 = Multiply(A=Add(A=Time(), B=ScalarParameter() {group=Shine, parameterName=DelayShineCycle}), B=@n94)
  n96 = Add(
   A=Rotator(
    Coordinate=MaterialFunctionCall() {materialFunction=&GetUserInterfaceUV}.Normalized UV
    Time=ScalarParameter() {defaultValue=0.15, group=Shine, parameterName=Shine Rotation}
   ) {centerX=0, centerY=0}
   B=AppendVector(A=OneMinus(None=@n95), B=Constant())
  )
  n97 = ScalarParameter() {defaultValue=0.75, group=Shine, parameterName=Shine SS Min}
  n98 = MaterialFunctionCall(
   Alpha=Sine(None=ComponentMask(None=@n96) {r=true}) {period=2}
   Min=@n97
   Max=Add(A=@n97, B=ScalarParameter() {defaultValue=0.25, group=Shine, parameterName=Shine SS Max})
  ) {materialFunction=&SmoothStep}
  n99 = Constant()
  n100 = Constant() {r=1}
  n101 = Constant()
  n102 = Add(A=Time(), B=ScalarParameter() {group=Shine, parameterName=DelayShineCycle})
  n103 = ScalarParameter() {defaultValue=3, group=Shine, parameterName=Cycles On}
  n104 = Multiply(A=Divide(A=ScalarParameter() {defaultValue=6, group=Shine, parameterName=Cycles Off}, B=@n94), B=Constant() {r=2})
  n105 = Add(A=Multiply(A=Divide(A=@n103, B=@n94), B=Constant() {r=2}), B=@n104)
  n106 = Constant() {r=-1}
  n107 = Add(A=Fmod(A=@n102, B=Constant() {r=1}), B=If(A=Fmod(A=@n102, B=@n105), B=@n104, "A > B"=Constant(), "A == B"=@n106, "A < B"=@n106)) {constB=-1}
  n108 = LinearInterpolate(A=@n28.RGB, B=Add(A=@n28.G, B=@n30.RGB), Alpha=ScalarParameter() {parameterName=UseTexture2})
  n109 = StaticSwitchParameter(
   True=ComponentMask(None=MaterialFunctionCall() {materialFunction=/Game/Athena/UI/Frontend/Art/MF_UIColorCollection_Frontend}.RGBA_BackingColor) {b=true, g=true, r=true}
   False=VectorParameter() {group=Color, parameterName=BackingColor, defaultValue=<rgba 0.25 0.25 0.25 1>}.RGB
  ) {parameterName=UseFrontendColorCollection}
  n110 = VectorParameter() {group=Color, parameterName=BackingHotspotColor, defaultValue=<rgba 0.5 0.5 0.5 1>}
  n111 = ScalarParameter() {defaultValue=1, group=Gradient, parameterName=Gradient_RadiusSize}
  n112 = ScalarParameter() {defaultValue=2.333, group=Gradient, parameterName=Gradient_Density}
  n113 = MaterialFunctionCall() {materialFunction=&GetUserInterfaceUV}
  n114 = MaterialFunctionCall(Float2=@n113.Pixel Size) {materialFunction=&BreakOutFloat2Components}
  n115 = Divide(A=@n114.R, B=@n114.R)
  n116 = AppendVector(A=@n115, B=Constant() {r=1})
  n117 = Multiply(A=OneMinus(None=@n115)) {constB=0.5}
  n118 = Constant()
  n119 = Add(
   A=Add(A=Multiply(A=@n113.Normalized UV, B=@n116), B=AppendVector(A=@n117, B=@n118))
   B=AppendVector(A=ScalarParameter() {group=Gradient, parameterName=Gradient_Offset_U}, B=ScalarParameter() {group=Gradient, parameterName=Gradient_Offset_V})
  )
  n120 = VectorParameter() {parameterName=HighlightPos, defaultValue=<rgba 0.5 1 1 1>}
  n121 = AppendVector(A=@n120.R, B=@n120.R)
  n122 = Multiply(
   A=StaticSwitchParameter(
    True=MaterialFunctionCall!(CenterPosition, Col_01, Col_02, Radius Scale=@n111, Density=@n112) {materialFunction=/Game/UI/Materials/MF_UI_SmoothPoly_RadialGradient}."Gradient (0-1)"
    False=MaterialFunctionCall(UVs=@n119, CenterPosition=@n121, Radius=@n111, Density=@n112) {materialFunction=&RadialGradientExponential}.RadialGradientExponential
   ) {group=Gradient, parameterName=UseSmoothPolyGradient}
   B=ScalarParameter() {defaultValue=0.5, group=Gradient, parameterName=Gradient_Amount}
  )
  n123 = VectorParameter() {group=Color, parameterName=BackingColor_Hover, defaultValue=<rgba 0.25 0.25 0.25 1>}
  n124 = VectorParameter() {group=Color, parameterName=BackingHotspotColor_Hover, defaultValue=<rgba 0.5 0.5 0.5 1>}
  n125 = ScalarParameter() {group=Animation, parameterName=Animate_BackingColor}
  n126 = StaticSwitchParameter(
   True=LinearInterpolate(
    A=LinearInterpolate(A=@n109, B=@n110.RGB, Alpha=@n122)
    B=LinearInterpolate(A=@n123.RGB, B=@n124.RGB, Alpha=@n122)
    Alpha=ScalarParameter() {group=Animation, parameterName=Animate_BackingColor}
   )
   False=LinearInterpolate(
    A=LinearInterpolate(A=@n109, B=@n123.RGB, Alpha=@n125)
    B=VectorParameter() {group=Color, parameterName=BackingColor_Selected, defaultValue=<rgba 1 1 1 1>}.RGB
    Alpha=ScalarParameter() {group=Animation, parameterName=Animate_BackingColor_Selected}
   )
  ) {group=Gradient, parameterName=HasRadialGradient}
  n127 = LinearInterpolate(
   A=LinearInterpolate(
    A=VectorParameter() {group=Color, parameterName=BorderColor, defaultValue=<rgba 1 1 1 1>}.RGB
    B=VectorParameter() {group=Color, parameterName=BorderColor_Hover, defaultValue=<rgba 1 1 1 1>}.RGB
    Alpha=ScalarParameter() {group=Animation, parameterName=Animate_BorderColor_Hover}
   )
   B=VectorParameter() {group=Color, parameterName=BorderColor_Selected, defaultValue=<rgba 1 1 1 1>}.RGB
   Alpha=ScalarParameter() {group=Animation, parameterName=Animate_BorderColor_Selected}
  )
  n128 = LinearInterpolate(A=@n127, B=VectorParameter() {group=Color, parameterName=BorderHotspotColor, defaultValue=<rgba 0.5 0.5 0.5 1>}.RGB, Alpha=@n122)
  n129 = StaticSwitchParameter(
   True=LinearInterpolate(A=@n126, B=StaticSwitchParameter(True=@n128, False=@n127) {group=Gradient, parameterName=HasBorderRadialGradient}, Alpha=@n51)
   False=@n126
  ) {defaultValue=true, group=Border, parameterName=HasBorder}
  n130 = TextureCoordinate()
  n131 = ScalarParameter() {group=Texture, parameterName=OffsetFade}
  n132 = Saturate(None=StaticComponentMaskParameter(None=Add(A=@n130, B=@n131)) {defaultG=true, group=Texture, parameterName=FadeDirection})
  n133 = Saturate(None=Power(Base=OneMinus(None=@n132), Exp=ScalarParameter() {defaultValue=1, group=Texture, parameterName=FadeSlope}))
  n134 = LinearInterpolate(
   A=LinearInterpolate(A=@n129, B=@n108, Alpha=@n49)
   B=LinearInterpolate(A=@n129, B=@n108, Alpha=DotProduct(A=@n133, B=@n49))
   Alpha=ScalarParameter() {group=Texture, parameterName=FadeAmount}
  )
  n135 = VectorParameter() {group=Texture, parameterName=TextureTint, defaultValue=<rgba 1 0 0 1>}
  n136 = StaticSwitchParameter(True=@n48, False=@n31.Result) {defaultValue=true, group=Border, parameterName=HasBorder}
  n137 = StaticSwitchParameter(
   True=LinearInterpolate(
    A=Saturate(None=@n134)
    B=LinearInterpolate(A=@n129, B=@n135.RGB, Alpha=@n136)
    Alpha=ScalarParameter() {group=Texture, parameterName=TextureTintAmount}
   )
   False=@n129
  ) {group=Texture, parameterName=HasTexture}
  n138 = Add(A=Multiply(A=LinearInterpolate(A=@n92, B=@n93.RGB, Alpha=@n98.Result), B=If(A=@n107, B=@n99, "A > B"=@n100, "A < B"=@n101)), B=@n137)
  n139 = Add(A=@n91, B=@n138)
  Saturate(None=StaticSwitchParameter(
    True=StaticSwitchParameter(True=@n139, False=@n138) {group=IsJustFill, parameterName=IsSimpleFill}
    False=StaticSwitchParameter(True=@n91, False=@n137) {group=IsJustFill, parameterName=IsSimpleFill}
   ) {group=Shine, parameterName=HasShineSwipe})
 ]
}
