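# Widget Blueprint for the respawn and spectate selection container in Athena LTM HUDs, derived from the native RespawnAndSpectateSelectContainerWidget.
# It has an empty event graph and no widgets of its own, so any behavior comes from the parent class.
# Its only addition is a RespawnTimerAdjustments transform whose default enlarges and shifts down, presumably for the respawn timer, though nothing here reads it.
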WidgetBlueprint /Game/Athena/HUD/LTM/RespawnAndSpectateSelect/RespawnAndSpectateSelectContainer
graphs [EventGraph {links [], nodes []}]
variables [RespawnTimerAdjustments {type=WidgetTransform, default {angle=0, scale=<1.5 1.5>, shear=<0 0>, translation=<0 100>}}]
widgets {
 info {inheritedWidgetCount=0, parentClass=/Script/FortniteUI.RespawnAndSpectateSelectContainerWidget, rootWidgetClass=null, widgetCount=0}
 widgets []
}
