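# UI sound cue for the HighTower LTM, named as the cue for a spawner becoming or being selectable.
# A mixer layers a player-survey pop-up wave at reduced level with an armory whoosh-back wave that starts after a short randomized delay; a third input is empty.
# The cue is boosted in volume, uses the non-critical in-game UI sound class and overrides concurrency to allow a single instance.
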SoundCue /Game/Sounds/LTM/HighTower/Hightower_UI_Spawner_Selectable_Cue
root=Mixer_0
nodes [
 {id=Delay_1, children [WavePlayer_1], properties {delayMax=0.1, delayMin=0.08}}
 {id=Mixer_0, children [Delay_1,WavePlayer_0,null], properties {inputVolume [1,0.7,1]}}
 {id=WavePlayer_0, properties {soundWaveAssetPtr=/Game/Sounds/Fort_UI_Sounds/FrontEnd/Fort_UI_PlayerSurvey_PopUp_01}}
 {id=WavePlayer_1, properties {soundWaveAssetPtr=/Game/Sounds/Fort_UI_Sounds/S12_Frontend/UI_S12_Armory_Skin_Options_Whoosh_Back}}
]
properties {
 bOverrideConcurrency=true
 maxDistance=2097152
 soundClassObject=&Fort_UI_InGame_NonCritical
 volumeMultiplier=2
 concurrencyOverrides {maxCount=1}
}
