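# Button widget for one target in the LTM respawn-and-spectate selection HUD, showing an avatar or icon.
# An invisible TargetButton forwards click, hover and focus to the native base, showing Confirm or spectating content beneath.
# Base-class events drive the checkmark, red eliminated X, highlight flicker and scale, swapping the angled background brush by selectability, highlight and icon type only when changed.
# A parallel button-style selection function exists but nothing here calls it.
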WidgetBlueprint /Game/Athena/HUD/LTM/RespawnAndSpectateSelect/RespawnAndSpectateSelectButton
graphs [
 DetermineButtonImageAndSet {nodes [
   DetermineButtonImageAndSet() {is=title K2Node_FunctionEntry}
   > n1 = Sequence()
   > n2 = Branch(Condition=AND Boolean(A=Get bIsSelectable() -> (bIsSelectable:bool), B=Get bHighlighted() -> (bHighlighted:bool)))
   > Set ButtonImageToSet(ButtonImageToSet:Object Reference=Select(
     Option 0:Object Reference=Get NonGradientHighlightImage() -> (NonGradientHighlightImage:Object Reference)
     Option 1:Object Reference=Get GradientHighlightImage() -> (GradientHighlightImage:Object Reference)
     Index=IsCharacterIcon!() -> (ReturnValue:bool)
    ) -> (ReturnValue:Object Reference)) -> (Output_Get:Object Reference)
   Set ButtonImageToSet(execute=@n2.else, ButtonImageToSet=Get DefaultImage() -> (DefaultImage:Object Reference))
   SetButtonImage(execute=@n1.then_1)
  ]}
 DetermineSelectableButtonStyleAndSet {nodes [
   DetermineSelectableButtonStyleAndSet() {is=title K2Node_FunctionEntry}
   > n1 = Sequence()
   > n2 = Branch(Condition=Get bIsSelectable())
   n3 = Get bHighlighted()
   > n4 = Branch(Condition=@n3)
   n5 = IsCharacterIcon()
   > Set ButtonStyleToSet(ButtonStyleToSet:Common Button Style Class Reference=Select(
     Option 0:Common Button Style Class Reference=Get ButtonStyle_StormIconHighlight() -> (ButtonStyle_StormIconHighlight:Common Button Style Class Reference)
     Option 1:Common Button Style Class Reference=Get ButtonStyle_DefaultHighlighted() -> (ButtonStyle_DefaultHighlighted:Common Button Style Class Reference)
     Index=@n5
    ) -> (ReturnValue:Common Button Style Class Reference)) -> (Output_Get:Common Button Style Class Reference)
   Set ButtonStyleToSet(
    execute=@n4.else
    ButtonStyleToSet=Select(
     Option 0=Get ButtonStyle_StormIconDefault() -> (ButtonStyle_StormIconDefault:Common Button Style Class Reference)
     Option 1=Get ButtonStyle_Default() -> (ButtonStyle_Default:Common Button Style Class Reference)
     Index=@n5
    )
   )
   Set ButtonStyleToSet(
    execute=@n2.else
    ButtonStyleToSet=Select(
     Option 0=Get ButtonStyle_UnSelectable() -> (ButtonStyle_UnSelectable:Common Button Style Class Reference)
     Option 1=Get ButtonStyle_UnSelectableHighlight() -> (ButtonStyle_UnSelectableHighlight:Common Button Style Class Reference)
     Index=@n3
    )
   )
   SetButtonStyle(execute=@n1.then_1)
  ]}
 EventGraph {nodes [
   "On Button Clicked (TargetButton)"!() -> (OutputDelegate, then, Button:Common Button Legacy Object Reference) {is=title K2Node_ComponentBoundEvent}
   > OnSelectionMade(self:Respawn And Spectate Select Button Widget Base Object Reference)
   "On Button Hovered (TargetButton)"!() -> (OutputDelegate, then, Button:Common Button Legacy Object Reference) {is=title K2Node_ComponentBoundEvent}
   n1 = GetOwningPlayer()
   > n2 = Branch(Condition=NOT Boolean(A=IsUsingTouch(PlayerController=@n1)))
   > OnSelectionHovered(self:Respawn And Spectate Select Button Widget Base Object Reference)
   > n3 = SetVisibility(self=Get Overlay_ActionContent() -> (Overlay_ActionContent:Overlay Object Reference), InVisibility=HitTestInvisible)
   PlaySound2D(execute=[@n3.then, @n2.else], Sound=Get UI_SFX_OnHovered() -> (UI_SFX_OnHovered:Sound Base Object Reference))
   "On Button Unhovered (TargetButton)"!() -> (OutputDelegate, then, Button:Common Button Legacy Object Reference) {is=title K2Node_ComponentBoundEvent}
   > Branch(Condition=NOT Boolean(A=IsMobileGame()))
   > SetVisibility(self=Get Overlay_ActionContent(), InVisibility=Hidden)
   OnFocused() {is=title K2Node_CustomEvent}
   > OnSelectionHovered()
   n4 = IsHovered(self=Get TargetButton() -> (TargetButton:Common Button Legacy Object Reference))
   n5 = "Equal (Enum)"(A=GetCurrentInputType(self=Get CommonInputSubsystem()), B=Gamepad)
   > Branch(Condition=AND Boolean(A=Get bIsSelectable(), B=OR Boolean(A=@n4, B=@n5)))
   > SetVisibility(self=Get Overlay_ActionContent(), InVisibility=HitTestInvisible)
   n6 = Event SetSelectionVisualsVisibility!() -> (OutputDelegate, then, bNewVisibility:bool) {is=title K2Node_Event}
   > SetVisibility(
    self=Get Image_SelectionConfirmed() -> (Image_SelectionConfirmed:Image Object Reference)
    InVisibility=Select(Option 0:ESlateVisibility Enum=Hidden, Option 1:ESlateVisibility Enum=SelfHitTestInvisible, Index=@n6.bNewVisibility) -> (ReturnValue:ESlateVisibility Enum)
   )
   Event IsSelectableUpdated_BP() {is=title K2Node_Event}
   > DetermineButtonImageAndSet() {is=title K2Node_CallFunction}
   > n7 = Branch(Condition=Get bIsSelectable())
   n8 = Get Switcher_ActionContent() -> (Switcher_ActionContent:Common Widget Switcher Legacy Object Reference)
   > SetActiveWidget(self:Widget Switcher Object Reference=@n8, Widget=Get Overlay_Respawn() -> (Overlay_Respawn:Overlay Object Reference))
   > PlaySound2D(Sound=Get UI_SFX_IsSelectable() -> (UI_SFX_IsSelectable:Sound Base Object Reference))
   n9 = IsMobileGame()
   > Branch(Condition=AND Boolean(A=IsHovered(), B=NOT Boolean(A=@n9)))
   > SetVisibility(self=Get Overlay_ActionContent(), InVisibility=HitTestInvisible)
   SetActiveWidget(execute=@n7.else, self=@n8, Widget=Get TextBlock_Spectating() -> (TextBlock_Spectating:Common Text Object Reference))
   n10 = Event SetHoverHighlightVisibility!() -> (OutputDelegate, then, bNewVisibility:bool) {is=title K2Node_Event}
   > Set bHighlighted(bHighlighted:bool=@n10.bNewVisibility) -> (Output_Get:bool)
   > n11 = Get!(execute:Exec) -> (then:Exec, else:Exec, SelectedFlickerImage:Image Object Reference)
   > SetVisibility(self=@n11, InVisibility=Select(Option 0=Hidden, Option 1=HitTestInvisible, Index=@n10.bNewVisibility))
   > DetermineButtonImageAndSet()
   > SetRenderScale!(
    execute
    self:Widget Object Reference
    Scale:Vector 2D Structure=Select(
     Option 0:Vector 2D Structure=Get UnSelectedScale() -> (UnSelectedScale:Vector 2D Structure)
     Option 1:Vector 2D Structure=Get SelectedScale() -> (SelectedScale:Vector 2D Structure)
     Index=@n10.bNewVisibility
    ) -> (ReturnValue:Vector 2D Structure)
   ) -> (then)
   n12 = Event OnIsEnabledUpdated_BP!() -> (OutputDelegate, then, bNewIsEnabled:bool) {is=title K2Node_Event}
   > n13 = Branch(Condition=@n12.bNewIsEnabled)
   n14 = Get EliminatedImage() -> (EliminatedImage:Image Object Reference)
   > SetVisibility(self=@n14, InVisibility=Hidden)
   > Branch(Condition=IsHovered(self=Get TargetButton()))
   n15 = Get Overlay_ActionContent()
   > SetVisibility(self=@n15, InVisibility=HitTestInvisible)
   SetVisibility(execute=@n13.else, self=@n14, InVisibility=SelfHitTestInvisible)
   > SetVisibility(self=@n15, InVisibility=Hidden)
   n16 = Event SetButtonVisibility_BP!() -> (OutputDelegate, then, bNewVisbility:bool) {is=title K2Node_Event}
   > SetVisibility(InVisibility=Select(Option 0:ESlateVisibility Enum=Collapsed, Option 1:ESlateVisibility Enum=SelfHitTestInvisible, Index=@n16.bNewVisbility) -> (ReturnValue:ESlateVisibility Enum))
   Event SetNavigationFocusOnButton() {is=title K2Node_Event}
   > Focus()
  ]}
 Focus {nodes [
   Focus() {is=title K2Node_FunctionEntry}
   > SetFocus(self=Get TargetButton())
   > OnFocused() {is=title K2Node_CallFunction}
  ]}
 IsCharacterIcon {nodes [
   IsCharacterIcon() {is=title K2Node_FunctionEntry}
   n1 = Get Switcher_TargetState() -> (Switcher_TargetState:Common Widget Switcher Legacy Object Reference)
   > n2 = Is Valid(InputObject=@n1)
   Return Node!(execute=@n2.Is Not Valid, ReturnValue:bool=false)
   > Return Node(ReturnValue="Equal (Integer)"(
     A=Get ActiveWidgetIndex!(self:Widget Switcher Object Reference=@n1) -> (ActiveWidgetIndex:int=0)
     B=Get TargetImageIndex() -> (TargetImageIndex:int=0)
    ) {is=title K2Node_CallFunction})
  ]}
 SetButtonImage {nodes [
   SetButtonImage() {is=title K2Node_FunctionEntry}
   n1 = Get ButtonImageToSet() -> (ButtonImageToSet:Object Reference)
   > Branch(Condition="Not Equal (Object)"!(A:Object Reference=@n1, B:Object Reference=Get CurrentButtonImage() -> (CurrentButtonImage:Object Reference)) -> (ReturnValue:bool))
   > Set CurrentButtonImage(CurrentButtonImage:Object Reference=@n1) -> (Output_Get:Object Reference)
   > n2 = Get!(execute) -> (then, else, ButtonImage:Image Object Reference)
   n3 = Break Slate Brush!(SlateBrush:"Slate Brush Structure (by ref)"=Get Brush!(self:Image Object Reference=@n2) -> (Brush:Slate Brush Structure)) -> (
    TintColor:Slate Color Structure
    DrawAs:ESlateBrushDrawType Enum
    Tiling:ESlateBrushTileType Enum
    Mirroring:ESlateBrushMirrorType Enum
    ImageSize:Vector 2D Structure
    Margin:Margin Structure
    ResourceObject:Object Reference
    OutlineSettings:Slate Brush Outline Settings Structure
   ) {is=title K2Node_BreakStruct}
   > SetBrush(
    self=@n2
    InBrush=Make Slate Brush!(
     TintColor:Slate Color Structure=@n3
     DrawAs:ESlateBrushDrawType Enum=@n3.DrawAs
     Tiling:ESlateBrushTileType Enum=@n3.Tiling
     Mirroring:ESlateBrushMirrorType Enum=@n3.Mirroring
     ImageSize:Vector 2D Structure=@n3.ImageSize
     Margin:Margin Structure=@n3.Margin
     ResourceObject:Object Reference=@n1
     OutlineSettings:Slate Brush Outline Settings Structure=(CornerRadii=(X=0,Y=0,Z=0,W=0),Color=(SpecifiedColor=(R=0,G=0,B=0,A=0),ColorUseRule=UseColor_Specified),Width=0,RoundingType=HalfHeightRadius,bUseBrushTransparency=False)
    ) -> (SlateBrush:Slate Brush Structure) {is=title K2Node_MakeStruct}
   )
  ]}
 SetButtonStyle {nodes [
   SetButtonStyle() {is=title K2Node_FunctionEntry}
   n1 = Get ButtonStyleToSet() -> (ButtonStyleToSet:Common Button Style Class Reference)
   > Branch(Condition="Not Equal (Class)"!(A:Object Class Reference=@n1, B:Object Class Reference=Get CurrentButtonStyle() -> (CurrentButtonStyle:Common Button Style Class Reference)) -> (ReturnValue:bool))
   > Set CurrentButtonStyle(CurrentButtonStyle:Common Button Style Class Reference=@n1) -> (Output_Get:Common Button Style Class Reference)
   > n2 = Get!(execute) -> (then, else, TargetButton:Common Button Legacy Object Reference)
   > SetStyle!(execute, self:Common Button Base Object Reference=@n2, InStyle:Common Button Style Class Reference=@n1) -> (then)
  ]}
]
variables [
 bHighlighted {type=boolean}
 ButtonImageToSet {type=Object, default=/Game/Athena/HUD/LTM/RespawnAndSpectateSelect/Materials/Athena_Angled_UnslectableHighlight}
 ButtonStyle_Default {
  type=Class
  default=&/Athena_ButtonStyle_RespawnAndSpectateSelectButton_Default.Athena_ButtonStyle_RespawnAndSpectateSelectButton_Default_C
 }
 ButtonStyle_DefaultHighlighted {
  type=Class
  default=&/Athena_ButtonStyle_RespawnAndSpectateSelectButton_DefaultHighlighted.Athena_ButtonStyle_RespawnAndSpectateSelectButton_DefaultHighlighted_C
 }
 ButtonStyle_Selectable {
  type=Class
  default=&/Athena_ButtonStyle_RespawnAndSpectateSelectButton_Selectable.Athena_ButtonStyle_RespawnAndSpectateSelectButton_Selectable_C
 }
 ButtonStyle_StormIconDefault {
  type=Class
  default=&/Athena_ButtonStyle_RespawnAndSpectateSelectButton_StormIconDefault.Athena_ButtonStyle_RespawnAndSpectateSelectButton_StormIconDefault_C
 }
 ButtonStyle_StormIconHighlight {
  type=Class
  default=&/Athena_ButtonStyle_RespawnAndSpectateSelectButton_StormIconHighlight.Athena_ButtonStyle_RespawnAndSpectateSelectButton_StormIconHighlight_C
 }
 ButtonStyle_UnSelectable {
  type=Class
  default=&/Athena_ButtonStyle_RespawnAndSpectateSelectButton_UnSelectableDefault.Athena_ButtonStyle_RespawnAndSpectateSelectButton_UnSelectableDefault_C
 }
 ButtonStyle_UnSelectableHighlight {
  type=Class
  default=&/Athena_ButtonStyle_RespawnAndSpectateSelectButton_UnSelectableHighlighted.Athena_ButtonStyle_RespawnAndSpectateSelectButton_UnSelectableHighlighted_C
 }
 ButtonStyleToSet {type=Class, default=null}
 CurrentButtonImage {type=Object, default=null}
 CurrentButtonStyle {type=Class, default=null}
 DefaultImage {type=Object, default=&Athena_Angled_UnslectableHighlight}
 GradientHighlightImage {type=Object, default=/Game/Athena/HUD/LTM/RespawnAndSpectateSelect/Materials/M_UI_AngledBox_Gradient_Inst}
 NonGradientHighlightImage {type=Object, default=/Game/Athena/HUD/LTM/RespawnAndSpectateSelect/Materials/Athena_Angled_StormHighlight}
 SelectedScale {type=Vector2D, default=<1.15 1.15>}
 TargetImageIndex {type=integer, default=1}
 UI_SFX_IsSelectable {type=SoundBase, default=/Game/Sounds/LTM/HighTower/Hightower_UI_Spawner_Selectable_Cue}
 UI_SFX_OnHovered {type=SoundBase, default=/Game/Sounds/LTM/HighTower/Hightower_UI_Spawner_Click_Cue}
 UnSelectedScale {type=Vector2D, default=<1 1>}
]
widgets {
 info {
  inheritedWidgetCount=7
  parentClass=/Script/FortniteUI.RespawnAndSpectateSelectButtonWidgetBase
  rootWidgetClass=&Overlay
  widgetCount=32
 }
 widgets [
  Overlay_2 {widgetClassPath=&Overlay, widget=Overlay_2}
  SelectedFlickerSizeBox {
   widgetClassPath=&SizeBox
   parent=Overlay_2
   widget=SelectedFlickerSizeBox
   properties {renderTransform {angle=0, scale=<1 1>, shear=<5.5 0>, translation=<0 0>}}
   slot {class=OverlaySlot, properties {horizontalAlignment=HAlign_Center, verticalAlignment=VAlign_Center}}
  }
  SelectedFlickerImage {
   widgetClassPath=&Image
   bIsVariable=true
   parent=SelectedFlickerSizeBox
   widget=SelectedFlickerImage
   properties {
    visibility=Hidden
    brush {
     bIsDynamicallyLoaded=false
     drawAs=Image
     imageType=NoImage
     mirroring=NoMirror
     resourceName=None
     resourceObject=/Game/Athena/HUD/LTM/RespawnAndSpectateSelect/Materials/M_UI_AngledBox_Gradient_Inst_Pulse
     tiling=NoTile
     imageSize=<310 228>
     margin=<ltrb 0 0 0 0>
     outlineSettings {
      bUseBrushTransparency=false
      roundingType=HalfHeightRadius
      width=0
      color {colorUseRule=UseColor_Specified, specifiedColor=<rgba 0 0 0 0>}
      cornerRadii=<0 0 0 0>
     }
     tintColor {colorUseRule=UseColor_Specified, specifiedColor=<rgba 1 1 1 1>}
     uVRegion {bIsValid=false, max=<0 0>, min=<0 0>}
    }
   }
   slot {class=SizeBoxSlot, properties {horizontalAlignment=HAlign_Center, verticalAlignment=VAlign_Center, padding=<ltrb -50 0 -50 0>}}
  }
  VerticalBox_0 {
   widgetClassPath=&VerticalBox
   parent=Overlay_2
   widget=VerticalBox_0
   slot {class=OverlaySlot, properties {horizontalAlignment=HAlign_Center, verticalAlignment=VAlign_Center}}
  }
  Overlay_ButtonContent {widgetClassPath=&Overlay, parent=VerticalBox_0, widget=Overlay_ButtonContent, slot {class=VerticalBoxSlot}}
  Overlay_1 {
   widgetClassPath=&Overlay
   parent=Overlay_ButtonContent
   widget=Overlay_1
   slot {class=OverlaySlot, properties {horizontalAlignment=HAlign_Fill, verticalAlignment=VAlign_Fill}}
  }
  SizeBox {
   widgetClassPath=&SizeBox
   parent=Overlay_1
   widget=SizeBox
   properties {renderTransform {angle=0, scale=<1.2 1.55>, shear=<0 0>, translation=<0 0>}}
   slot {class=OverlaySlot, properties {horizontalAlignment=HAlign_Fill, verticalAlignment=VAlign_Center}}
  }
  HoverHighlight {
   widgetClassPath=&Image
   bIsVariable=true
   parent=SizeBox
   widget=HoverHighlight
   properties {
    visibility=Hidden
    brush {
     bIsDynamicallyLoaded=false
     drawAs=Box
     imageType=NoImage
     mirroring=NoMirror
     resourceName=None
     resourceObject=/Game/Athena/HUD/LTM/RespawnAndSpectateSelect/Materials/Athena_Angled_Highlight
     tiling=NoTile
     imageSize=<72 72>
     margin=<ltrb 0.5 0.5 0.5 0.5>
     outlineSettings {
      bUseBrushTransparency=false
      roundingType=HalfHeightRadius
      width=0
      color {colorUseRule=UseColor_Specified, specifiedColor=<rgba 0 0 0 0>}
      cornerRadii=<0 0 0 0>
     }
     tintColor {colorUseRule=UseColor_Specified, specifiedColor=<rgba 0.9131 0.9131 0.9131 1>}
     uVRegion {bIsValid=false, max=<0 0>, min=<0 0>}
    }
    renderTransform {angle=0, scale=<1.05 1.55>, shear=<0 0>, translation=<0 0>}
   }
   slot {class=SizeBoxSlot}
  }
  SizeBox_82 {
   widgetClassPath=&SizeBox
   parent=Overlay_1
   widget=SizeBox_82
   properties {renderOpacity=0, renderTransform {angle=0, scale=<0.77 0.95>, shear=<-7 0>, translation=<0 0>}}
   slot {class=OverlaySlot, properties {horizontalAlignment=HAlign_Center, verticalAlignment=VAlign_Center}}
  }
  TargetButton {
   widgetClassPath=/Game/Athena/UI/Matchmaking/SimpleCommonButton
   bInherited=true
   bIsVariable=true
   parent=SizeBox_82
   widget=TargetButton
   properties {
    bIsFocusable=true
    onButtonBaseClicked=(&self:WidgetTree.TargetButton.HandleOnButtonClicked)
    onButtonBaseDoubleClicked=(&self:WidgetTree.TargetButton.HandleOnButtonDoubleClicked)
    onButtonBaseHovered=(&self:WidgetTree.TargetButton.HandleOnButtonHovered)
    onButtonBaseUnhovered=(&self:WidgetTree.TargetButton.HandleOnButtonUnhovered)
    onSelectedChangedBase=(&self:WidgetTree.TargetButton.HandleOnSelectedChanged)
    style=&Athena_ButtonStyle_RespawnAndSpectateSelectButton_Default_C
    visibility=Visible
   }
   slot {class=SizeBoxSlot}
  }
  ButtonImage {
   widgetClassPath=&Image
   bIsVariable=true
   parent=Overlay_1
   widget=ButtonImage
   properties {
    visibility=HitTestInvisible
    brush {
     bIsDynamicallyLoaded=false
     drawAs=Image
     imageType=NoImage
     mirroring=NoMirror
     resourceName=None
     resourceObject=&Athena_Angled_UnslectableHighlight
     tiling=NoTile
     imageSize=<230 170>
     margin=<ltrb 0 0 0 0>
     outlineSettings {
      bUseBrushTransparency=false
      roundingType=HalfHeightRadius
      width=0
      color {colorUseRule=UseColor_Specified, specifiedColor=<rgba 0 0 0 0>}
      cornerRadii=<0 0 0 0>
     }
     tintColor {colorUseRule=UseColor_Specified, specifiedColor=<rgba 1 1 1 1>}
     uVRegion {bIsValid=false, max=<0 0>, min=<0 0>}
    }
    renderTransform {angle=0, scale=<1 1>, shear=<5 0>, translation=<0 0>}
   }
   slot {class=OverlaySlot}
  }
  VerticalBox_50 {
   widgetClassPath=&VerticalBox
   parent=Overlay_ButtonContent
   widget=VerticalBox_50
   slot {class=OverlaySlot, properties {horizontalAlignment=HAlign_Center, verticalAlignment=VAlign_Center}}
  }
  Overlay_0 {
   widgetClassPath=&Overlay
   parent=VerticalBox_50
   widget=Overlay_0
   slot {class=VerticalBoxSlot, properties {horizontalAlignment=HAlign_Center, verticalAlignment=VAlign_Center}}
  }
  Switcher_TargetState {
   widgetClassPath=&CommonWidgetSwitcherLegacy
   bInherited=true
   bIsVariable=true
   parent=Overlay_0
   widget=Switcher_TargetState
   slot {class=OverlaySlot, properties {horizontalAlignment=HAlign_Center, verticalAlignment=VAlign_Center}}
  }
  DefaultButtonImageSizeBox {
   widgetClassPath=&SizeBox
   bInherited=true
   bIsVariable=true
   parent=Switcher_TargetState
   widget=DefaultButtonImageSizeBox
   properties {bOverride_HeightOverride=true, bOverride_WidthOverride=true, heightOverride=100, widthOverride=100}
   slot {class=WidgetSwitcherSlot, properties {horizontalAlignment=HAlign_Center, verticalAlignment=VAlign_Center, padding=<ltrb 0 4 0 0>}}
  }
  DefaultButtonImage {
   widgetClassPath=&Image
   bInherited=true
   bIsVariable=true
   parent=DefaultButtonImageSizeBox
   widget=DefaultButtonImage
   properties {
    visibility=SelfHitTestInvisible
    brush {
     bIsDynamicallyLoaded=false
     drawAs=Image
     imageType=NoImage
     mirroring=NoMirror
     resourceName=None
     resourceObject=null
     tiling=NoTile
     imageSize=<128 128>
     margin=<ltrb 0 0 0 0>
     outlineSettings {
      bUseBrushTransparency=false
      roundingType=HalfHeightRadius
      width=0
      color {colorUseRule=UseColor_Specified, specifiedColor=<rgba 0 0 0 0>}
      cornerRadii=<0 0 0 0>
     }
     tintColor {colorUseRule=UseColor_Specified, specifiedColor=<rgba 1 1 1 1>}
     uVRegion {bIsValid=false, max=<0 0>, min=<0 0>}
    }
   }
   slot {class=SizeBoxSlot, properties {horizontalAlignment=HAlign_Center, verticalAlignment=VAlign_Center}}
  }
  TargetImageSizeBox {
   widgetClassPath=&SizeBox
   bInherited=true
   bIsVariable=true
   parent=Switcher_TargetState
   widget=TargetImageSizeBox
   properties {bOverride_HeightOverride=true, bOverride_WidthOverride=true, heightOverride=145, widthOverride=145}
   slot {class=WidgetSwitcherSlot, properties {horizontalAlignment=HAlign_Center, verticalAlignment=VAlign_Center, padding=<ltrb 0 0 0 4>}}
  }
  TargetImage {
   widgetClassPath=/Script/FortniteUI.FortSocialAvatarIcon
   bInherited=true
   bIsVariable=true
   parent=TargetImageSizeBox
   widget=TargetImage
   properties {brush {
     bIsDynamicallyLoaded=false
     drawAs=Image
     imageType=NoImage
     mirroring=NoMirror
     resourceName=None
     resourceObject=null
     tiling=NoTile
     imageSize=<128 128>
     margin=<ltrb 0 0 0 0>
     outlineSettings {
      bUseBrushTransparency=false
      roundingType=HalfHeightRadius
      width=0
      color {colorUseRule=UseColor_Specified, specifiedColor=<rgba 0 0 0 0>}
      cornerRadii=<0 0 0 0>
     }
     tintColor {colorUseRule=UseColor_Specified, specifiedColor=<rgba 1 1 1 1>}
     uVRegion {bIsValid=false, max=<0 0>, min=<0 0>}
    }}
   slot {class=SizeBoxSlot, properties {horizontalAlignment=HAlign_Center, verticalAlignment=VAlign_Center}}
  }
  Image_SelectionConfirmed {
   widgetClassPath=&Image
   bIsVariable=true
   parent=Overlay_0
   widget=Image_SelectionConfirmed
   properties {
    visibility=Hidden
    brush {
     bIsDynamicallyLoaded=false
     drawAs=Image
     imageType=NoImage
     mirroring=NoMirror
     resourceName=None
     resourceObject=/Game/Athena/HUD/Bacchus/HUDLayoutTool/HLT_Icons/T-HLT-CheckMark-32
     tiling=NoTile
     imageSize=<32 32>
     margin=<ltrb 0 0 0 0>
     outlineSettings {
      bUseBrushTransparency=false
      roundingType=HalfHeightRadius
      width=0
      color {colorUseRule=UseColor_Specified, specifiedColor=<rgba 0 0 0 0>}
      cornerRadii=<0 0 0 0>
     }
     tintColor {colorUseRule=UseColor_Specified, specifiedColor=<rgba 0 1 0 1>}
     uVRegion {bIsValid=false, max=<0 0>, min=<0 0>}
    }
   }
   slot {class=OverlaySlot, properties {horizontalAlignment=HAlign_Right, verticalAlignment=VAlign_Bottom, padding=<ltrb 0 0 10 10>}}
  }
  TargetNameTextWidget {
   widgetClassPath=&CommonTextBlock
   bInherited=true
   bIsVariable=true
   parent=VerticalBox_0
   widget=TargetNameTextWidget
   properties {
    autoWrapText=true
    justification=Center
    scrollStyle=&TextScrollStyle-Base_C
    text=Defaultdefaultdefault default
    visibility=Collapsed
    wrapTextAt=150
    font {
     bForceMonospaced=false
     bMaterialIsStencil=false
     fontMaterial=null
     fontObject=&BurbankSmall
     letterSpacing=0
     monospacedWidth=1
     size=16
     skewAmount=0
     typefaceFontName=Bold
     outlineSettings {
      bApplyOutlineToDropShadows=false
      bMiteredCorners=false
      bSeparateFillAlpha=false
      outlineMaterial=null
      outlineSize=0
      outlineColor=<rgba 0 0 0 1>
     }
    }
    shadowColorAndOpacity=<rgba 0 0 0 0.4>
    shadowOffset=<3 3>
   }
   slot {class=VerticalBoxSlot, properties {horizontalAlignment=HAlign_Center, verticalAlignment=VAlign_Bottom}}
  }
  Spacer_95 {widgetClassPath=&Spacer, parent=VerticalBox_0, widget=Spacer_95, properties {size=<175 1>}, slot {class=VerticalBoxSlot}}
  Overlay_ActionContent {
   widgetClassPath=&Overlay
   bIsVariable=true
   parent=Overlay_2
   widget=Overlay_ActionContent
   properties {visibility=Hidden}
   slot {class=OverlaySlot, properties {horizontalAlignment=HAlign_Center, verticalAlignment=VAlign_Bottom, padding=<ltrb 0 0 0 -20>}}
  }
  Switcher_ActionContent {
   widgetClassPath=&CommonWidgetSwitcherLegacy
   bIsVariable=true
   parent=Overlay_ActionContent
   widget=Switcher_ActionContent
   slot {class=OverlaySlot, properties {horizontalAlignment=HAlign_Center}}
  }
  TextBlock_Spectating {
   widgetClassPath=&CommonTextBlock
   bIsVariable=true
   parent=Switcher_ActionContent
   widget=TextBlock_Spectating
   properties {
    autoWrapText=true
    justification=Center
    textTransformPolicy=ToUpper
    visibility=Visible
    wrapTextAt=300
    font {
     bForceMonospaced=false
     bMaterialIsStencil=false
     fontMaterial=null
     fontObject=&Burbank
     letterSpacing=0
     monospacedWidth=1
     size=28
     skewAmount=0
     typefaceFontName=BurbankRegular-Black
     outlineSettings {
      bApplyOutlineToDropShadows=false
      bMiteredCorners=false
      bSeparateFillAlpha=false
      outlineMaterial=null
      outlineSize=1
      outlineColor=<rgba 0.0153 0 0.0156 0.5>
     }
    }
    renderTransform {angle=0, scale=<1 1>, shear=<-15 0>, translation=<0 0>}
    shadowColorAndOpacity=<rgba 0 0 0 0.4>
    shadowOffset=<4 4>
   }
   slot {class=WidgetSwitcherSlot, properties {padding=<ltrb 0 10 0 0>}}
  }
  Overlay_Respawn {
   widgetClassPath=&Overlay
   bIsVariable=true
   parent=Switcher_ActionContent
   widget=Overlay_Respawn
   slot {class=WidgetSwitcherSlot, properties {horizontalAlignment=HAlign_Center, padding=<ltrb 0 10 0 0>}}
  }
  CommonVisibilityWidget_0 {
   widgetClassPath=&CommonVisibilityWidgetLegacy
   parent=Overlay_Respawn
   widget=CommonVisibilityWidget_0
   properties {bShowForAndroid=false, bShowForIOS=false, bShowForTouch=false, hiddenType=Hidden}
   slot {class=OverlaySlot, properties {horizontalAlignment=HAlign_Fill, verticalAlignment=VAlign_Fill}}
  }
  HorizontalBox_83 {widgetClassPath=&HorizontalBox, parent=CommonVisibilityWidget_0, widget=HorizontalBox_83, slot {class=BorderSlot}}
  InputActionWidget {
   widgetClassPath=&CommonActionWidget
   parent=HorizontalBox_83
   widget=InputActionWidget
   properties {
    progressMaterialParam=Percentage
    inputActions [{dataTable=&InputActionDataTable, rowName=GenericForward}]
    progressMaterialBrush {
     bIsDynamicallyLoaded=false
     drawAs=Image
     imageType=NoImage
     mirroring=NoMirror
     resourceName=None
     resourceObject=/Game/UI/Materials/HollowCircularProgressBar
     tiling=NoTile
     imageSize=<36 36>
     margin=<ltrb 0 0 0 0>
     outlineSettings {
      bUseBrushTransparency=false
      roundingType=HalfHeightRadius
      width=0
      color {colorUseRule=UseColor_Specified, specifiedColor=<rgba 0 0 0 0>}
      cornerRadii=<0 0 0 0>
     }
     tintColor {colorUseRule=UseColor_Specified, specifiedColor=<rgba 1 1 0 1>}
     uVRegion {bIsValid=false, max=<0 0>, min=<0 0>}
    }
   }
   slot {class=HorizontalBoxSlot, properties {horizontalAlignment=HAlign_Center, verticalAlignment=VAlign_Center, padding=<ltrb 4 0 4 0>}}
  }
  TextBlock_Respawn {
   widgetClassPath=&CommonTextBlock
   bIsVariable=true
   parent=HorizontalBox_83
   widget=TextBlock_Respawn
   properties {
    autoWrapText=true
    justification=Center
    text=Confirm
    textTransformPolicy=ToUpper
    visibility=Visible
    wrapTextAt=300
    font {
     bForceMonospaced=false
     bMaterialIsStencil=false
     fontMaterial=null
     fontObject=&Burbank
     letterSpacing=0
     monospacedWidth=1
     size=28
     skewAmount=0
     typefaceFontName=BurbankRegular-Black
     outlineSettings {
      bApplyOutlineToDropShadows=false
      bMiteredCorners=false
      bSeparateFillAlpha=false
      outlineMaterial=null
      outlineSize=1
      outlineColor=<rgba 0.0153 0 0.0156 0.5>
     }
    }
    renderTransform {angle=0, scale=<1 1>, shear=<-15 0>, translation=<0 0>}
    shadowColorAndOpacity=<rgba 0 0 0 0.4>
    shadowOffset=<4 4>
   }
   slot {class=HorizontalBoxSlot, properties {padding=<ltrb 0 0 4 0>}}
  }
  Spacer_90 {
   widgetClassPath=&Spacer
   parent=Overlay_ActionContent
   widget=Spacer_90
   properties {size=<220 1>}
   slot {class=OverlaySlot, properties {horizontalAlignment=HAlign_Center}}
  }
  EliminatedSizeBox {
   widgetClassPath=&SizeBox
   parent=Overlay_2
   widget=EliminatedSizeBox
   properties {bOverride_HeightOverride=true, bOverride_WidthOverride=true, heightOverride=90, widthOverride=90}
   slot {class=OverlaySlot, properties {horizontalAlignment=HAlign_Center, verticalAlignment=VAlign_Center}}
  }
  EliminatedImage {
   widgetClassPath=&Image
   bIsVariable=true
   parent=EliminatedSizeBox
   widget=EliminatedImage
   properties {
    brush {
     bIsDynamicallyLoaded=false
     drawAs=Image
     imageType=NoImage
     mirroring=NoMirror
     resourceName=None
     resourceObject=/Game/Athena/HUD/LTM/RespawnAndSpectateSelect/Materials/T_UI_X_Icon_128
     tiling=NoTile
     imageSize=<128 128>
     margin=<ltrb 0 0 0 0>
     outlineSettings {
      bUseBrushTransparency=false
      roundingType=HalfHeightRadius
      width=0
      color {colorUseRule=UseColor_Specified, specifiedColor=<rgba 0 0 0 0>}
      cornerRadii=<0 0 0 0>
     }
     tintColor {colorUseRule=UseColor_Specified, specifiedColor=<rgba 1 1 1 1>}
     uVRegion {bIsValid=false, max=<0 0>, min=<0 0>}
    }
    colorAndOpacity=<rgba 1 0 0 1>
   }
   slot {class=SizeBoxSlot}
  }
 ]
}
