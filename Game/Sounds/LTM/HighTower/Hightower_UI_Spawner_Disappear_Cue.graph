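# Sound cue for a HighTower LTM spawner-disappear UI sound, routed to the non-critical in-game UI sound class.
# A mixer layers the Creative item spawner spawn sound at full level with a quieter lock-device accept sound.
# Each layer passes through its own short fixed delay, so the lock sound trails the spawn sound slightly.
# A concurrency override allows one instance at a time and blocks new plays while it sounds.
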SoundCue /Game/Sounds/LTM/HighTower/Hightower_UI_Spawner_Disappear_Cue
root=Mixer_0
nodes [
 {id=Delay_0, children [WavePlayer_0], properties {delayMax=0.1, delayMin=0.1}}
 {id=Delay_2, children [WavePlayer_1], properties {delayMax=0.02, delayMin=0.02}}
 {id=Mixer_0, children [Delay_2,Delay_0], properties {inputVolume [1,0.65]}}
 {id=WavePlayer_0, properties {soundWaveAssetPtr=/Game/Sounds/Creative/Toys/Locks/LockDevice_Accept_01}}
 {id=WavePlayer_1, properties {soundWaveAssetPtr=/Game/Sounds/Creative/Toys/ItemSpawner/ItemSpawner_ItemSpawn}}
]
properties {
 bOverrideConcurrency=true
 maxDistance=2097152
 priority=0.9
 soundClassObject=&Fort_UI_InGame_NonCritical
 volumeMultiplier=1.35
 concurrencyOverrides {maxCount=1, resolutionRule=PreventNew}
}
