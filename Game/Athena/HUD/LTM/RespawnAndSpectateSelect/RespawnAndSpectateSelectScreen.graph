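# Widget Blueprint for an LTM screen to choose a respawn location, with four indexed select buttons, header text and a hidden Confirm button.
# On construct it gathers the buttons into an array for its native RespawnAndSpectateSelectWidgetBase parent and binds input-method changes, which refocus the native desired focus target.
# StartClosingWidget deactivates the screen and plays a close sound.
# HandleBackInput only prints a debug string and returns false for pass-through.
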WidgetBlueprint /Game/Athena/HUD/LTM/RespawnAndSpectateSelect/RespawnAndSpectateSelectScreen
graphs [
 Close Panel {nodes [
   Close Panel() {is=title K2Node_FunctionEntry}
   > DeactivateWidget()
  ]}
 EventGraph {nodes [
   HandleInputMethodChanged!() -> (OutputDelegate, then, bNewInputType:ECommonInputType Enum) {is=title K2Node_CustomEvent}
   > SetInitialFocus()
   Event Destruct()
   n1 = Create Event()
   > Unbind Event from On Input Method Changed(self=Get CommonInputSubsystem(), Delegate=@n1)
   Event Construct()
   > Bind Event to On Input Method Changed(self=Get CommonInputSubsystem(), Delegate=@n1)
   > Set RespawnAndSpectateSelectButtons(RespawnAndSpectateSelectButtons:Array of Respawn And Spectate Select Button Object References=Make Array!(
     "[0]":Respawn And Spectate Select Button Object Reference=Get RespawnAndSpectateSelectButton_0() -> (RespawnAndSpectateSelectButton_0:Respawn And Spectate Select Button Object Reference)
     "[1]":Respawn And Spectate Select Button Object Reference=Get RespawnAndSpectateSelectButton_1() -> (RespawnAndSpectateSelectButton_1:Respawn And Spectate Select Button Object Reference)
     "[2]":Respawn And Spectate Select Button Object Reference=Get RespawnAndSpectateSelectButton_2() -> (RespawnAndSpectateSelectButton_2:Respawn And Spectate Select Button Object Reference)
     "[3]":Respawn And Spectate Select Button Object Reference=Get RespawnAndSpectateSelectButton_3() -> (RespawnAndSpectateSelectButton_3:Respawn And Spectate Select Button Object Reference)
    ) -> (Array:Array of Respawn And Spectate Select Button Object References)) -> (Output_Get:Array of Respawn And Spectate Select Button Object References)
   Event StartClosingWidget() {is=title K2Node_Event}
   > Close Panel() {is=title K2Node_CallFunction}
   > PlaySound2D(Sound=Get UI_SFX_CloseScreen() -> (UI_SFX_CloseScreen:Sound Base Object Reference))
  ]}
 HandleBackInput {nodes [
   HandleBackInput() {is=title K2Node_FunctionEntry}
   > PrintString!(execute, InString="Back Pressed!", bPrintToScreen, bPrintToLog, TextColor=(R=1,G=0.0085,B=0,A=1), Duration=20, Key) -> (then)
   > Return Node!(execute, bPassThrough:bool=false)
  ]}
 SetInitialFocus {nodes [
   SetInitialFocus() {is=title K2Node_FunctionEntry}
   > n1 = Cast To RespawnAndSpectateSelectButton!(
    execute
    Object=NativeGetDesiredFocusTarget!(self:Respawn And Spectate Select Widget Base Object Reference) -> (ReturnValue:Widget Object Reference)
   ) -> (then, CastFailed, AsRespawn and Spectate Select Button:Respawn And Spectate Select Button Object Reference) {is=title K2Node_DynamicCast}
   > Focus(self:Respawn And Spectate Select Button Object Reference=@n1)
  ]}
]
variables [
 RespawnAndSpectateSelectButtons {type="array<RespawnAndSpectateSelectButton_C>", default []}
 UI_SFX_CloseScreen {type=SoundBase, default=/Game/Sounds/LTM/HighTower/Hightower_UI_Spawner_Disappear_Cue}
]
widgets {
 info {
  inheritedWidgetCount=3
  parentClass=/Script/FortniteUI.RespawnAndSpectateSelectWidgetBase
  rootWidgetClass=&Overlay
  widgetCount=15
 }
 widgets [
  Overlay_23 {widgetClassPath=&Overlay, widget=Overlay_23}
  Background {
   widgetClassPath=&Image
   bIsVariable=true
   parent=Overlay_23
   widget=Background
   properties {
    visibility=SelfHitTestInvisible
    brush {
     bIsDynamicallyLoaded=false
     drawAs=Image
     imageType=NoImage
     mirroring=NoMirror
     resourceName=None
     resourceObject=null
     tiling=NoTile
     imageSize=<32 32>
     margin=<ltrb 0 0 0 0>
     outlineSettings {
      bUseBrushTransparency=false
      roundingType=HalfHeightRadius
      width=0
      color {colorUseRule=UseColor_Specified, specifiedColor=<rgba 0 0 0 0>}
      cornerRadii=<0 0 0 0>
     }
     tintColor {colorUseRule=UseColor_Specified, specifiedColor=<rgba 1 1 1 0>}
     uVRegion {bIsValid=false, max=<0 0>, min=<0 0>}
    }
   }
   slot {class=OverlaySlot, properties {horizontalAlignment=HAlign_Fill, verticalAlignment=VAlign_Fill}}
  }
  VerticalBox_0 {
   widgetClassPath=&VerticalBox
   parent=Overlay_23
   widget=VerticalBox_0
   slot {class=OverlaySlot, properties {horizontalAlignment=HAlign_Center, verticalAlignment=VAlign_Bottom, padding=<ltrb 0 0 0 151>}}
  }
  TextHeader {
   widgetClassPath=/Script/UMG.RichTextBlock
   bInherited=true
   bIsVariable=true
   parent=VerticalBox_0
   widget=TextHeader
   properties {
    justification=Center
    text=Choose Respawn Location
    textStyleSet=&/RespawnSelectUIStyles
    textTransformPolicy=ToUpper
    wrapTextAt=1225
    renderTransform {angle=0, scale=<1 1>, shear=<-15 0>, translation=<0 0>}
   }
   slot {class=VerticalBoxSlot, properties {verticalAlignment=VAlign_Bottom}}
  }
  TextDescription {
   widgetClassPath=&RichTextBlock
   bInherited=true
   bIsVariable=true
   parent=VerticalBox_0
   widget=TextDescription
   properties {
    justification=Center
    text="[PH] Hover over options to spectate targets."
    textStyleSet=&RespawnSelectUIStyles
    visibility=Collapsed
    wrapTextAt=1225
   }
   slot {class=VerticalBoxSlot, properties {verticalAlignment=VAlign_Bottom}}
  }
  SizeBox_0 {
   widgetClassPath=&SizeBox
   parent=VerticalBox_0
   widget=SizeBox_0
   properties {bOverride_HeightOverride=true, bOverride_WidthOverride=true, heightOverride=253.9456, widthOverride=1115.8647}
   slot {class=VerticalBoxSlot, properties {horizontalAlignment=HAlign_Center, verticalAlignment=VAlign_Bottom}}
  }
  HorizontalBox_73 {
   widgetClassPath=&HorizontalBox
   parent=SizeBox_0
   widget=HorizontalBox_73
   slot {class=SizeBoxSlot, properties {horizontalAlignment=HAlign_Center}}
  }
  RespawnAndSpectateSelectButton_0 {
   widgetClassPath=&/RespawnAndSpectateSelectButton
   bIsVariable=true
   parent=HorizontalBox_73
   widget=RespawnAndSpectateSelectButton_0
   properties {
    buttonIndex=0
    navigation=&self:WidgetTree.RespawnAndSpectateSelectButton_0.WidgetNavigation_0
   }
   slot {class=HorizontalBoxSlot, properties {horizontalAlignment=HAlign_Center, verticalAlignment=VAlign_Top, padding=<ltrb -10 0 -10 0>}}
  }
  RespawnAndSpectateSelectButton_1 {
   widgetClassPath=&RespawnAndSpectateSelectButton
   bIsVariable=true
   parent=HorizontalBox_73
   widget=RespawnAndSpectateSelectButton_1
   properties {
    buttonIndex=1
    navigation=&self:WidgetTree.RespawnAndSpectateSelectButton_1.WidgetNavigation_0
   }
   slot {class=HorizontalBoxSlot, properties {horizontalAlignment=HAlign_Center, verticalAlignment=VAlign_Top, padding=<ltrb -10 0 -10 0>}}
  }
  RespawnAndSpectateSelectButton_2 {
   widgetClassPath=&RespawnAndSpectateSelectButton
   bIsVariable=true
   parent=HorizontalBox_73
   widget=RespawnAndSpectateSelectButton_2
   properties {
    buttonIndex=2
    navigation=&self:WidgetTree.RespawnAndSpectateSelectButton_2.WidgetNavigation_0
   }
   slot {class=HorizontalBoxSlot, properties {horizontalAlignment=HAlign_Center, verticalAlignment=VAlign_Top, padding=<ltrb -10 0 -10 0>}}
  }
  RespawnAndSpectateSelectButton_3 {
   widgetClassPath=&RespawnAndSpectateSelectButton
   bIsVariable=true
   parent=HorizontalBox_73
   widget=RespawnAndSpectateSelectButton_3
   properties {
    buttonIndex=3
    navigation=&self:WidgetTree.RespawnAndSpectateSelectButton_3.WidgetNavigation_0
   }
   slot {class=HorizontalBoxSlot, properties {horizontalAlignment=HAlign_Center, verticalAlignment=VAlign_Top, padding=<ltrb -10 0 -10 0>}}
  }
  HorizontalBox_98 {
   widgetClassPath=&HorizontalBox
   parent=VerticalBox_0
   widget=HorizontalBox_98
   slot {class=VerticalBoxSlot, properties {padding=<ltrb 0 -30 0 0>}}
  }
  Spacer_50 {
   widgetClassPath=&Spacer
   parent=HorizontalBox_98
   widget=Spacer_50
   slot {class=HorizontalBoxSlot, properties {size {sizeRule=Fill, value=1.5}}}
  }
  ConfirmButton {
   widgetClassPath=/Game/UI/Foundation/Modals/ConfirmationWindow/Athena_ConfirmationButton
   bInherited=true
   bIsVariable=true
   parent=HorizontalBox_98
   widget=ConfirmButton
   properties {
    buttonText=Confirm
    minHeight=75
    onButtonBaseClicked=(&self:WidgetTree.ConfirmButton.HandleOnButtonClicked)
    onButtonBaseDoubleClicked=(&self:WidgetTree.ConfirmButton.HandleOnButtonDoubleClicked)
    onButtonBaseHovered=(&self:WidgetTree.ConfirmButton.HandleOnButtonHovered)
    onButtonBaseUnhovered=(&self:WidgetTree.ConfirmButton.HandleOnButtonUnhovered)
    onSelectedChangedBase=(&self:WidgetTree.ConfirmButton.HandleOnSelectedChanged)
    visibility=Hidden
    renderTransform {angle=0, scale=<1 1>, shear=<-10 0>, translation=<0 0>}
    triggeringInputAction {dataTable=&InputActionDataTable, rowName=Confirm}
   }
   slot {class=HorizontalBoxSlot, properties {size {sizeRule=Fill, value=2}}}
  }
  Spacer_85 {
   widgetClassPath=&Spacer
   parent=HorizontalBox_98
   widget=Spacer_85
   slot {class=HorizontalBoxSlot, properties {size {sizeRule=Fill, value=1.5}}}
  }
 ]
}
